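# Additive, unlit, two-sided glow material for Niagara mesh particles in the Voltaic Huntress character's FX, named for a tube glow.
# Particle color feeds emissive through a shared adjustment function.
# The opacity chain multiplies a powered UV sphere mask, a view-angle fade with sensitivity driven by vertex red, and particle alpha.
# It also applies depth fade, which a quality switch replaces with a constant on Low and Medium.
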Material /Game/Characters/Player/Female/Medium/Bodies/F_MED_VoltaicHuntress/FX/M_Tube_Glow
graph {
 properties {
  bUsedWithNiagaraMeshParticles=true
  blendMode=BLEND_Additive
  pixelDepthOffsetMode=PDOM_Legacy
  shadingModel=MSM_Unlit
  twoSided=true
  usedShadingModels=Unlit
 }
 nodes [
  n1 = ParticleColor()
  MaterialFunctionCall(Emissive=@n1.RGB) {materialFunction=&MF_FX_EmissiveAdjustments}
  n2 = Power(
   Base=SphereMask(A=TextureCoordinate(), B=Constant() {r=0.5}, Radius=Constant() {r=0.5}, Hardness=Constant())
   Exp=ScalarParameter() {defaultValue=2, parameterName=Power}
  )
  n3 = MaterialFunctionCall("FadeSensitivity (optional)"=LinearInterpolate(A=Constant() {r=18}, B=Constant() {r=2}, Alpha=VertexColor().R)) {materialFunction=&FadeBasedOnViewAngle}
  n4 = DepthFade(FadeDistance=ScalarParameter() {defaultValue=2, parameterName=DepthFade Distance})
  n5 = Constant() {r=1}
  Multiply(A=Multiply(A=Multiply(A=@n2, B=@n3.Result), B=QualitySwitch(Default=@n4, Low=@n5, High=@n4, Medium=@n5, Epic=@n4)), B=@n1.A)
 ]
}
